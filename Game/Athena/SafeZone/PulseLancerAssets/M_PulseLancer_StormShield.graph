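# Additive, two-sided unlit material for the Athena safe zone storm shield wall, drawing animated energy lines, stars and noise.
# Cylindrical and world-space UVs feed scrolling textures, with safe-zone radius scaling tiling and depth, height and distance masks shaping the glow.
# Custom depth, offset by polar-coordinate noise, cuts a "Demon Rupture" hole with corrupted edges.
# The "BR?" switch and PC-Perf/Android platform switches zero out the rupture, while BR adds time-of-day line intensity.
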Material /Game/Athena/SafeZone/PulseLancerAssets/M_PulseLancer_StormShield
graph {
 properties {bAutomaticallySetUsageInEditor=false, blendMode=BLEND_Additive, shadingModel=MSM_Unlit, twoSided=true, usedShadingModels=Unlit}
 nodes [
  n1 = OneMinus(None=NamedRerouteUsage())
  n2 = Multiply(A=@n1) {constB=20}
  Add(A=@n2) {constB=0}
  n3 = Constant()
  n4 = MaterialFunctionCall(
   Input=CollectionParameter() {collection=/Game/Packages/Fortress_SharedMaterials/GlobalMaterialParameters/DirectSafeZoneMPC, parameterName=SafeRadius}
   Input Low=@n3
   Input High=Constant() {r=185185}
   Target Low=@n3
   Target High=Constant() {r=1}
  ) {materialFunction=&RemapValueRange}
  AppendVector(
   A=LinearInterpolate(
    A=ScalarParameter() {defaultValue=0.05, group=Visual Adjustments, parameterName=Scale Texture Repition Min}
    B=ScalarParameter() {defaultValue=10, group=Visual Adjustments, parameterName=Scale Texture Repition Max}
    Alpha=NamedRerouteDeclaration(None=@n4.Result) {name=Scale Storm Shield, nodeColor=<rgba 0 0.25 1 1>}
   ) {constA=0.01, constB=6}
   B=Constant() {r=1}
  )
  Constant() {r=0.7}
  Constant()
  Constant() {r=1}
  Divide(B=ScalarParameter() {defaultValue=1, parameterName=Rupture Tile Y})
  Divide()
  n5 = ScalarParameter() {defaultValue=0.3659, group=Demon Rupture, parameterName=Line corruption Intensity}
  n6 = ScalarParameter() {defaultValue=653.4991, group=Demon Rupture, parameterName=Rupture Bias Offset}
  n7 = ScreenPosition()
  n8 = Constant()
  n9 = Multiply(
   A=NamedRerouteUsage()
   B=LinearInterpolate(
    A=ScalarParameter() {defaultValue=68.2513, group=Demon Rupture, parameterName=Corruption Tile Min}
    B=ScalarParameter() {defaultValue=809.843, group=Demon Rupture, parameterName=Corruption Tile Max}
    Alpha=NamedRerouteUsage()
   )
  )
  n10 = Multiply(A=Time()) {constB=-0.04}
  n11 = NamedRerouteUsage()
  n12 = ScalarParameter() {defaultValue=52.0892, group=Demon Rupture, parameterName=Rupture Distortion TileSize}
  n13 = Time()
  n14 = TextureSample(UVs=Add(A=Multiply(A=@n11, B=@n12), B=Divide(A=@n13) {constB=30})) {samplerType=SAMPLERTYPE_LinearGrayscale, texture=&/T_Perlin_HQ16_SafeZone}
  n15 = Add(A=@n14.G) {constB=-0.5}
  n16 = MaterialFunctionCall!(UV Coord=Add(A=Add(A=@n9, B=@n10), B=@n15), LenghtScale=Constant() {r=1}, RadialScale) {materialFunction=/Game/Global/MaterialFunctions/TechArt/MF_PolarCoordinates}
  n17 = Divide(A=ComponentMask(None=@n16.Result) {r=true}, B=ScalarParameter() {defaultValue=0.8658, group=Demon Rupture, parameterName=Rupture Tile X})
  n18 = ComponentMask(None=@n16.Result) {g=true}
  n19 = ScalarParameter() {defaultValue=0.8689, group=Demon Rupture, parameterName=Rupture Distortion Intensity}
  n20 = TextureSample(UVs=Add(A=AppendVector(A=@n17, B=@n18), B=Multiply(A=@n19, B=@n15))) {samplerType=SAMPLERTYPE_LinearGrayscale, texture=&T_Perlin_HQ16_SafeZone}
  n21 = Add(A=@n20.G) {constB=-0.5}
  n22 = ScalarParameter() {group=Demon Rupture, parameterName=Min Distortion}
  n23 = ScalarParameter() {defaultValue=0.05, group=Demon Rupture, parameterName=Max Distortion}
  n24 = ScreenPosition()
  n25 = Round(None=Divide(A=SceneTexture(UVs=@n24.ViewportUV) {sceneTextureId=PPI_CustomDepth}.Color, B=Constant() {r=34704.0547}))
  n26 = Constant()
  n27 = If(A=OneMinus(None=Saturate(None=@n25)), "A > B"=Constant() {r=1}, "A == B"=@n26, "A < B"=@n26)
  n28 = StaticSwitchParameter(True=@n8, False=Multiply(A=Add(A=@n21, B=@n15), B=LinearInterpolate(A=@n22, B=@n23, Alpha=@n27))) {parameterName=BR?}
  n29 = MaterialFunctionCall!("Xb1,PS4"=@n28, "PC-High&&Epic,XSX,PS5"=@n28, "PC-Perf,Android"=@n8, Switch=@n28, "PC-Low(NotPerf)&&Medium"=@n28) {materialFunction=/Game/Effects/Niagara/MasterMaterials/Functions/Common/PlatformSwitch_3D}
  n30 = Subtract(A=PixelDepth().R, B=SceneTexture(UVs=Add(A=@n7.ViewportUV, B=@n29.Result)) {sceneTextureId=PPI_CustomDepth}.Color)
  n31 = Divide(
   A=ComponentMask(None=Add(A=@n6, B=@n30)) {r=true}
   B=ScalarParameter() {defaultValue=690, group=Demon Rupture, parameterName=Rupture Mask Size}
  )
  Divide(A=@n31, B=@n5)
  n32 = WorldPosition() {worldPositionShaderOffset=WPT_CameraRelative}
  n33 = Constant()
  If(
   A=MaterialFunctionCall(Vector 2=ComponentMask(None=@n32.XYZ) {g=true, r=true}) {materialFunction=/Engine/Functions/Engine_MaterialFunctions02/Utility/VectorLength}.V2 Length
   B=Multiply(
    A=ScalarParameter() {defaultValue=102045.8516, group=Visual Adjustments, parameterName=FarFade WPO}
    B=ScalarParameter() {defaultValue=1.1, group=Visual Adjustments, parameterName=Buffer}
   ) {constB=1.36, desc=add some buffer}
   "A > B"=Constant3Vector() {constant=<rgba 0 0 -999999 0>}
   "A == B"=@n33
   "A < B"=@n33
  )
  n34 = ScalarParameter() {parameterName="Debug Switch (0 or 1)", sliderMax=1}
  LinearInterpolate(B=Saturate(None=@n4.Result), Alpha=Saturate(None=Sign(None=@n34)))
  Multiply(A=CollectionParameter() {collection=&DirectSafeZoneMPC, parameterName=SafeRadius}) {constB=0.5}
  n35 = PixelDepth()
  n36 = Divide(A=@n35.R, B=ScalarParameter() {defaultValue=74537.7344, group=Visual Adjustments, parameterName=Max distance Fade})
  n37 = Multiply(A=@n36, B=ScalarParameter() {defaultValue=2.3597, group=Stars, parameterName=Star Fade Distance}) {constB=0.5}
  Multiply(A=@n37, B=@n37)
  n38 = NamedRerouteUsage()
  n39 = WorldPosition()
  n40 = ComponentMask(None=@n39.XYZ) {b=true, r=true}
  n41 = ComponentMask(None=@n39.XYZ) {b=true, g=true}
  n42 = ScalarParameter() {defaultValue=700, group=Stars, parameterName=Noise Stars Min}
  n43 = ScalarParameter() {defaultValue=2000, group=Stars, parameterName=Noise Stars Max}
  n44 = NamedRerouteUsage()
  n45 = Add(A=Multiply(A=NamedRerouteUsage()) {constB=0.65}, B=Divide(A=AppendVector(A=@n40, B=@n41), B=LinearInterpolate(A=@n42, B=@n43, Alpha=@n44)))
  n46 = ComponentMask(None=@n45) {g=true, r=true}
  n47 = Add(A=Multiply(A=ComponentMask(None=@n45) {a=true, b=true}) {constB=1.2}, B=Constant() {r=0.375}) {constB=0.37}
  n48 = Abs(None=VertexNormalWS())
  n49 = Saturate(None=MaterialFunctionCall!(In=Multiply(A=@n48, B=@n48), Contrast=Constant() {r=5}) {materialFunction=/Engine/Functions/Engine_MaterialFunctions01/ImageAdjustment/CheapContrast_RGB}.Result)
  MaterialExpressionLinearInterpolate_15 = LinearInterpolate(
   A=TextureSample(UVs=VertexInterpolator(VS=@n46).PS) {texture=&/T_StarSky_SafeZone}.RGBA
   B=TextureSample(UVs=VertexInterpolator(VS=@n47).PS) {texture=&T_StarSky_SafeZone}.RGBA
   Alpha=VertexInterpolator(VS=ComponentMask(None=@n49) {r=true}).PS
  )
  MaterialExpressionReroute_2 = Reroute(None=@MaterialExpressionLinearInterpolate_15)
  n50 = Convert(None=@MaterialExpressionReroute_2) {
   nodeName=Break Float4
   convertInputs [{
     type=Vector4
     defaultValue=<rgba 0 0 0 1>
     expressionInput {
      expression=&self:MaterialExpressionReroute_2
      inputName=None
      mask=0
      maskA=0
      maskB=0
      maskG=0
      maskR=0
      outputIndex=0
     }
    }]
   convertOutputs [{type=Scalar}, {type=Scalar}, {type=Scalar}, {type=Scalar}]
  }
  n51 = Multiply(A=Multiply(A=NamedRerouteUsage(), B=NamedRerouteUsage()), B=@n50)
  n52 = NamedRerouteUsage()
  n53 = Multiply(
   A=Multiply(A=NamedRerouteUsage(), B=LinearInterpolate(A=@n38, B=@n51, Alpha=@n52) {constB=0})
   B=ScalarParameter() {defaultValue=30, group=Stars, parameterName=Big Stars Intensity}
  )
  Multiply(A=@n53)
  n54 = CameraVectorWS()
  n55 = PixelNormalWS()
  n56 = OneMinus(None=Abs(None=DotProduct(A=@n54, B=@n55)))
  n57 = Divide(A=PixelDepth().R) {constB=200}
  n58 = Constant() {r=0.4}
  n59 = NamedRerouteUsage()
  n60 = ScalarParameter() {defaultValue=-0.1447, group=Pos Mask BG Color, parameterName=Top Fade Pos}
  n61 = Saturate(None=Multiply(A=Add(A=@n59, B=@n60), B=ScalarParameter() {defaultValue=31.8189, group=Pos Mask BG Color, parameterName=Top fade Size}))
  n62 = NamedRerouteUsage()
  n63 = NamedRerouteUsage()
  n64 = Saturate(None=OneMinus(None=Subtract(A=@n62, B=@n63)))
  n65 = NamedRerouteUsage()
  n66 = Constant() {r=8.7953}
  n67 = Multiply(A=Add(A=Multiply(A=@n65, B=@n66), B=NamedRerouteUsage()), B=NamedRerouteUsage())
  n68 = NamedRerouteUsage()
  n69 = NamedRerouteUsage()
  n70 = Multiply(A=@n69) {constB=0.6}
  n71 = Add(
   A=Add(A=NamedRerouteUsage(), B=NamedRerouteUsage())
   B=LinearInterpolate(A=ScalarParameter() {defaultValue=0.75, group=Pos Mask BG Color, parameterName=Pruple Area Bias}, Alpha=NamedRerouteUsage()) {constB=0.9}
  )
  n72 = Multiply(
   A=MaterialFunctionCall("Xb1,PS4"=@n70, "PC-High&&Epic,XSX,PS5"=Multiply(A=@n69) {constB=1.3}, "PC-Perf,Android"=@n70, Switch=@n69, "PC-Low(NotPerf)&&Medium"=@n70) {materialFunction=&PlatformSwitch_1D}.Result
   B=LinearInterpolate(A=NamedRerouteUsage(), B=NamedRerouteUsage(), Alpha=Saturate(None=@n71))
  )
  n73 = Multiply(A=NamedRerouteUsage(), B=NamedRerouteUsage())
  n74 = Saturate(None=Subtract(A=NamedRerouteUsage(), B=NamedRerouteUsage()))
  n75 = Power(Base=@n74, Exp=Constant() {r=5.9539})
  n76 = Add(A=@n75, B=Multiply(A=@n74, B=NamedRerouteUsage()))
  n77 = NamedRerouteUsage()
  n78 = Add(A=@n75) {constB=0.05}
  n79 = NamedRerouteUsage()
  n80 = Add(A=Add(A=Multiply(A=@n76, B=@n77), B=Multiply(A=@n78, B=@n79)), B=NamedRerouteUsage())
  n81 = ScalarParameter() {defaultValue=30, group=Lines, parameterName=Intensity}
  n82 = CollectionParameter() {
   collection=&FortniteMaterialParameters
   parameterName=FortniteActiveTimeOfDay
  }
  n83 = NamedRerouteDeclaration(None=Saturate(None=MaterialFunctionCall(Float4=@n82) {materialFunction=&BreakOutFloat4Components}.A)) {name=NightTime, nodeColor=<rgba 0 1 0.6016 1>}
  n84 = StaticSwitchParameter(True=Multiply(A=@n81, B=LinearInterpolate(Alpha=@n83) {constA=1.7, constB=1.2}), False=@n81) {parameterName=BR?}
  n85 = Add(
   A=Multiply(A=Multiply(A=@n67, B=@n68), B=ScalarParameter() {defaultValue=0.224, group=Lines, parameterName=Secondary Lines Intensity})
   B=Add(A=Add(A=@n72, B=@n73), B=Multiply(A=@n80, B=@n84))
  )
  n86 = Add(A=NamedRerouteUsage(), B=Multiply(A=NamedRerouteUsage(), B=NamedRerouteUsage()))
  n87 = Multiply(A=@n86) {constB=2.5}
  n88 = Multiply(A=@n87, B=@n87)
  n89 = NamedRerouteUsage()
  n90 = Constant() {r=7}
  n91 = Add(A=@n87, B=Multiply(A=@n86, B=Multiply(A=@n89, B=@n90)))
  n92 = VectorParameter() {group=Color, parameterName=Demon Rupture Color, defaultValue=<rgba 0.6562 0 0.1521 1>}
  n93 = Add(A=Multiply(A=Multiply(A=@n85, B=@n64), B=OneMinus(None=@n88)), B=Saturate(None=Multiply(A=@n91, B=@n92.RGB)))
  n94 = Multiply(A=NamedRerouteUsage(), B=Constant() {r=1.5943})
  n95 = NamedRerouteUsage()
  n96 = Multiply(A=Multiply(A=Saturate(None=@n94), B=OneMinus(None=@n95)), B=NamedRerouteUsage())
  n97 = StaticSwitchParameter(True=@n85, False=Add(A=Saturate(None=@n93), B=Multiply(A=@n96, B=@n92.RGB))) {parameterName=BR?}
  n98 = NamedRerouteUsage()
  n99 = Multiply(
   A=Multiply(A=Multiply(A=@n61, B=@n97), B=OneMinus(None=@n98))
   B=StaticSwitchParameter(True=Constant() {r=1}, False=NamedRerouteUsage()) {parameterName=BR?}
  )
  Multiply(
   A=OneMinus(None=Power(Base=@n56) {constExponent=10})
   B=Multiply(A=Saturate(None=@n57), B=EyeAdaptationInverse(LightValueInput=@n99, AlphaInput=@n58).EyeAdaptationInverse)
  )
  Multiply()
  Multiply()
  n100 = NamedRerouteUsage()
  Multiply(A=@n100) {constB=0.5}
  n101 = Constant()
  n102 = Saturate(None=@n31)
  n103 = Ceil(None=OneMinus(None=@n102))
  n104 = Add(A=@n20.G, B=Constant() {r=0.5032})
  n105 = SmoothStep(
   Min=ScalarParameter() {defaultValue=0.7885, group=Demon Rupture, parameterName=Rupture Contrast Min}
   Max=ScalarParameter() {defaultValue=2.5139, group=Demon Rupture, parameterName=Rupture Contrast Max}
   Value=Multiply(A=Multiply(A=@n103, B=@n102), B=Saturate(None=@n104))
  )
  n106 = Constant()
  n107 = Multiply(A=@n105, B=Add(A=@n105, B=StaticSwitchParameter(True=@n106, False=@n14.G) {parameterName=BR?}))
  n108 = StaticSwitchParameter(True=@n101, False=Saturate(None=Add(A=@n105, B=@n107))) {parameterName=BR?}
  NamedRerouteDeclaration(None=MaterialFunctionCall("Xb1,PS4"=@n108, "PC-High&&Epic,XSX,PS5"=@n108, "PC-Perf,Android"=@n101, Switch=@n108, "PC-Low(NotPerf)&&Medium"=@n108) {materialFunction=&PlatformSwitch_3D}.Result) {name=Demon Rupture, nodeColor=<rgba 0.8281 1 0 1>}
  n109 = Constant()
  n110 = Saturate(None=Multiply(A=@n31, B=@n5) {constB=0.8})
  n111 = StaticSwitchParameter(True=@n109, False=Multiply(A=@n110, B=@n110)) {parameterName=BR?}
  NamedRerouteDeclaration(None=MaterialFunctionCall("Xb1,PS4"=@n111, "PC-High&&Epic,XSX,PS5"=@n111, "PC-Perf,Android"=@n109, Switch=@n111, "PC-Low(NotPerf)&&Medium"=@n111) {materialFunction=&PlatformSwitch_3D}.Result) {name=Line Color mask, nodeColor=<rgba 0.6172 0 1 1>}
  n112 = NamedRerouteUsage()
  n113 = NamedRerouteDeclaration(None=Divide(A=Time(), B=ScalarParameter() {defaultValue=37.7194, group=Noises, parameterName=Wave Inensity Sped}) {constB=20}) {name=Time, nodeColor=<rgba 1 0 0.0769 1>}
  n114 = TextureSample(UVs=Divide(A=Add(A=@n112, B=@n113), B=ScalarParameter() {defaultValue=1.6, group=Noises, parameterName=Noise 1 Size})) {texture=&T_StarSky_SafeZone}
  n115 = Add(A=@n114.R) {constB=0.2}
  n116 = Multiply(A=@n112) {constB=0.672}
  n117 = NamedRerouteUsage()
  n118 = TextureSample(UVs=Divide(A=Add(A=@n116, B=@n117), B=ScalarParameter() {defaultValue=0.8, group=Noises, parameterName=Noise 2 Size})) {texture=&T_StarSky_SafeZone}
  n119 = Multiply(A=Multiply(A=@n115, B=@n114.B), B=@n118.R)
  n120 = Add(A=@n118.R) {constB=0.2}
  n121 = NamedRerouteDeclaration(None=Multiply(A=@n115, B=Multiply(A=@n120, B=@n118.B))) {name=Noise, nodeColor=<rgba 1 0.9609 0 1>}
  n122 = Multiply(
   A=StaticSwitchParameter(True=LinearInterpolate(Alpha=@n121) {constA=0.4, constB=0.6}, False=@n121) {parameterName=Intensity on BR only}
   B=ScalarParameter() {defaultValue=1.2, group=Noises, parameterName=Base Noise intensity}
  ) {constB=0.1}
  n123 = Multiply(A=@n122, B=@n122)
  n124 = Multiply(
   A=Add(A=NamedRerouteDeclaration(None=@n119) {name=Moving Noise, nodeColor=<rgba 1 0.3984 0 1>}) {constB=0.05}
   B=Multiply(A=NamedRerouteUsage(), B=@n123)
  )
  n125 = NamedRerouteUsage()
  n126 = Add(
   A=StaticSwitchParameter(True=LinearInterpolate(Alpha=@n125) {constA=0.2}, False=@n125) {parameterName=Intensity on BR only}
   B=ScalarParameter() {defaultValue=1.15, group=Noises, parameterName=Small Waves Intensity}
  )
  NamedRerouteDeclaration(None=Clamp(None=Multiply(A=@n124, B=@n126), Max=ScalarParameter() {defaultValue=1})) {name=Emissive Noise Color, nodeColor=<rgba 0.0078 1 0 1>}
  n127 = Convert(None=@MaterialExpressionLinearInterpolate_15) {
   nodeName=Break Float4
   convertInputs [{
     type=Vector4
     defaultValue=<rgba 0 0 0 1>
     expressionInput {
      expression=&self:MaterialExpressionLinearInterpolate_15
      inputName=None
      mask=0
      maskA=0
      maskB=0
      maskG=0
      maskR=0
      outputIndex=0
     }
    }]
   convertOutputs [{type=Scalar}, {type=Scalar}, {type=Scalar}, {type=Scalar}]
  }
  n128 = Multiply(A=Multiply(A=@n50, B=@n50), B=Add(A=@n127) {constB=0.4})
  n129 = Multiply(A=Multiply(A=@n128, B=@n128), B=Constant() {r=2}) {constB=10}
  n130 = Constant() {r=0.4}
  NamedRerouteDeclaration(None=Saturate(None=Add(A=@n129, B=@n130) {constB=0.1})) {name=Small Noise, nodeColor=<rgba 1 0.8906 0 1>}
  n131 = Constant()
  n132 = Saturate(None=Add(A=@n31, B=Constant() {r=-0.9893}))
  n133 = Power(Base=OneMinus(None=@n132), Exp=Constant() {r=44.5636}) {constExponent=30}
  n134 = Ceil(None=@n132)
  n135 = StaticSwitchParameter(True=@n131, False=Saturate(None=LinearInterpolate(B=@n133, Alpha=@n134) {constA=1})) {parameterName=BR?}
  NamedRerouteDeclaration(None=MaterialFunctionCall("Xb1,PS4"=@n135, "PC-High&&Epic,XSX,PS5"=@n135, "PC-Perf,Android"=@n131, Switch=@n135, "PC-Low(NotPerf)&&Medium"=@n135) {materialFunction=&PlatformSwitch_3D}.Result) {name=Fire Hole Mask, nodeColor=<rgba 0 1 0.1797 1>}
  n136 = CollectionParameter() {collection=&FortniteMaterialParameters, parameterName=AthenaSafeZoneRadius}
  NamedRerouteDeclaration(None=@n136) {name=Radius Storm Shield, nodeColor=<rgba 0 0.25 1 1>}
  n137 = Multiply(A=Add(A=NamedRerouteUsage())) {constB=0.5}
  n138 = Multiply(A=@n123, B=Multiply(A=@n53, B=Saturate(None=@n137)))
  n139 = VectorParameter() {group=Color, parameterName=Stars, defaultValue=<rgba 0.4975 0.2865 1 1>}
  NamedRerouteDeclaration(None=Multiply(A=NamedRerouteUsage(), B=Multiply(A=@n138, B=@n139.RGB))) {name=Stars, nodeColor=<rgba 0.7344 0 1 1>}
  n140 = Constant()
  n141 = StaticSwitchParameter(True=@n140, False=Saturate(None=@n134)) {parameterName=BR?}
  NamedRerouteDeclaration(None=MaterialFunctionCall("Xb1,PS4"=@n141, "PC-High&&Epic,XSX,PS5"=@n141, "PC-Perf,Android"=@n140, Switch=@n141, "PC-Low(NotPerf)&&Medium"=@n141) {materialFunction=&PlatformSwitch_3D}.Result) {name=Wall Hole, nodeColor=<rgba 1 0.375 0 1>}
  n142 = Divide(A=Time(), B=ScalarParameter() {defaultValue=242.494, group=Lines, parameterName=Line Movement}) {constB=40}
  n143 = WorldPosition()
  n144 = ComponentMask(None=@n143.XYZ) {b=true, r=true}
  n145 = ComponentMask(None=@n143.XYZ) {b=true, g=true}
  n146 = ScalarParameter() {defaultValue=5611.0054, group=Lines, parameterName=Tile Distortion}
  n147 = ScalarParameter() {defaultValue=124853.0859, group=Lines, parameterName=Tile Distortion Far}
  n148 = NamedRerouteUsage()
  n149 = Add(A=@n142, B=Divide(A=AppendVector(A=@n144, B=@n145), B=LinearInterpolate(A=@n146, B=@n147, Alpha=@n148)))
  n150 = Constant() {r=0.525}
  n151 = AppendVector(A=@n150, B=Multiply(A=@n150) {constB=0.5})
  n152 = VertexInterpolator(VS=Multiply(A=@n151, B=ComponentMask(None=@n149) {g=true, r=true}))
  n153 = Multiply(A=ComponentMask(None=@n149) {a=true, b=true}) {constB=1.5}
  n154 = Constant() {r=0.176}
  n155 = VertexInterpolator(VS=Multiply(A=@n151, B=Add(A=@n153, B=@n154) {constB=0.37}))
  n156 = Multiply(
   A=Multiply(
    A=TextureSample(UVs=@n152.PS) {samplerType=SAMPLERTYPE_LinearGrayscale, texture=&T_Perlin_HQ16_SafeZone}.G
    B=TextureSample(UVs=@n155.PS) {samplerType=SAMPLERTYPE_LinearGrayscale, texture=&T_Perlin_HQ16_SafeZone}.G
   )
   B=LinearInterpolate(
    A=ScalarParameter() {defaultValue=40, group=Lines, parameterName=Distortion Intensity Close}
    B=ScalarParameter() {defaultValue=20, group=Lines, parameterName=Distortion Intensity Far}
    Alpha=NamedRerouteUsage()
   ) {constB=35}
  )
  n157 = LinearInterpolate(
   A=ScalarParameter() {defaultValue=134.1031, group=Lines, parameterName=Tlling Lines}
   B=ScalarParameter() {defaultValue=2089.5786, group=Lines, parameterName=Tilling Lines Max Scale}
   Alpha=@n100
  )
  n158 = Add(A=@n156, B=VertexInterpolator(VS=Divide(A=@n145, B=@n157)).PS)
  n159 = NamedRerouteUsage()
  n160 = NamedRerouteUsage()
  n161 = Multiply(A=LinearInterpolate(B=Add(A=@n159, B=@n160), Alpha=NamedRerouteUsage()), B=Constant() {r=0.2}) {constB=100}
  NamedRerouteDeclaration(None=TextureSample(UVs=Add(A=@n158, B=@n161)) {samplerType=SAMPLERTYPE_LinearGrayscale, texture=&/T_GradientLine_HQ16_SafeZone}.G) {name=Lines, nodeColor=<rgba 0.8984 1 0 1>}
  n162 = DepthFade(
   Opacity=Constant() {r=1}
   FadeDistance=ScalarParameter() {defaultValue=197.4958, group=Visual Adjustments, parameterName=Depth fade Distance}
  )
  n163 = Multiply(
   A=Saturate(None=OneMinus(None=@n162))
   B=ScalarParameter() {defaultValue=2.5, group=Visual Adjustments, parameterName=Contact line Intensity Depth}
  ) {constB=2}
  NamedRerouteDeclaration(None=Multiply(A=@n163, B=@n163)) {name=Depth, nodeColor=<rgba 0.4297 0 1 1>}
  n164 = LocalPosition()
  n165 = ScalarParameter() {defaultValue=133.5986, group=Pos Mask BG Color, parameterName=Pos Mask}
  n166 = OneMinus(None=Divide(A=Add(A=@n164.Z, B=@n165), B=ScalarParameter() {defaultValue=190.8016, group=Pos Mask BG Color, parameterName=Pos Mask Fade}))
  n167 = Multiply(A=Add(A=NamedRerouteUsage()) {constB=0.2}, B=ScalarParameter() {defaultValue=3, group=Pos Mask BG Color, parameterName=Pos Mask Noise})
  n168 = Multiply(A=@n167, B=LinearInterpolate(A=Constant() {r=2}, Alpha=NamedRerouteUsage()) {constA=2, constB=1.2})
  NamedRerouteDeclaration(None=Multiply(A=Saturate(None=@n166), B=StaticSwitchParameter(True=@n168, False=@n167) {parameterName=BR?})) {name=Pos Mask Noise, nodeColor=<rgba 0.4062 1 0 1>}
  NamedRerouteDeclaration(None=VectorParameter() {group=Color, parameterName=Color, defaultValue=<rgba 0 0.1524 1 1>}.RGB) {name=Hon Color, nodeColor=<rgba 0 0.4924 1 1>}
  NamedRerouteDeclaration(None=VectorParameter() {group=Color, parameterName=Demon Color, defaultValue=<rgba 0.641 0 1 1>}.RGB) {name=Demon Color, nodeColor=<rgba 1 0 0.7266 1>}
  NamedRerouteDeclaration(None=@n166) {name=Pos Mask, nodeColor=<rgba 0 1 0.9297 1>}
  n169 = Power(Base=NamedRerouteUsage(), Exp=ScalarParameter() {defaultValue=0.526, group=Lines, parameterName=Secondary Lines Contrast}) {constExponent=0.2}
  n170 = Add(
   A=Multiply(A=OneMinus(None=@n169), B=ScalarParameter() {defaultValue=3.8124, group=Lines, parameterName=Secondary Lines Mult})
   B=ScalarParameter() {defaultValue=1.6643, group=Lines, parameterName=Secondary Lines Pos}
  )
  n171 = Multiply(A=OneMinus(None=NamedRerouteUsage()), B=Saturate(None=Sine(None=@n170)))
  n172 = OneMinus(None=NamedRerouteUsage())
  n173 = Multiply(A=Multiply(A=@n171, B=@n171), B=Multiply(A=@n172, B=@n172))
  NamedRerouteDeclaration(None=Add(A=@n173, B=Power(Base=@n173) {constExponent=8})) {name=Secondary Lines, nodeColor=<rgba 0.1016 0 1 1>}
  NamedRerouteDeclaration(None=@n142) {name=Time Slow, nodeColor=<rgba 0 1 0.2969 1>}
  NamedRerouteDeclaration(None=OneMinus(None=Saturate(None=@n37))) {name=star distance fade, nodeColor=<rgba 1 0 0.1641 1>}
  n174 = OneMinus(None=@n50)
  n175 = Multiply(A=@n50, B=Multiply(A=@n127, B=ScalarParameter() {defaultValue=50, group=Stars, parameterName=Mini Star intensity}))
  NamedRerouteDeclaration(None=Multiply(A=Multiply(A=@n174, B=@n175), B=@n1)) {name=MiniStars, nodeColor=<rgba 0.0547 1 0 1>}
  NamedRerouteDeclaration(None=@n114.R) {name=Noise 2, nodeColor=<rgba 0 0.5078 1 1>}
  n176 = Divide(A=@n35.R, B=ScalarParameter() {defaultValue=4058.4822, group=Visual Adjustments, parameterName=Distance fade})
  n177 = Constant() {r=0.7}
  NamedRerouteDeclaration(None=Saturate(None=MaterialFunctionCall(In=@n176, Contrast=@n177) {materialFunction=&CheapContrast}.Result)) {name=Distance Mask, nodeColor=<rgba 0 1 0.8594 1>}
  NamedRerouteDeclaration(None=@n156) {name=Noise HQ, nodeColor=<rgba 0 1 0.0156 1>}
  n178 = Constant()
  n179 = Constant() {r=-27928.2363}
  n180 = Subtract(
   A=WorldPosition().XYZ
   B=Add(
    A=ObjectPositionWS()
    B=MaterialFunctionCall(R=@n178, G=@n178, B=@n179) {materialFunction=&AppendMany}.RGB
   )
  )
  n181 = Divide(A=Arctangent2Fast(Y=ComponentMask(None=@n180) {g=true}, X=ComponentMask(None=@n180) {r=true})) {constB=3.1416}
  n182 = ObjectRadius()
  n183 = LinearInterpolate(
   A=ScalarParameter() {defaultValue=0.2, group=UVs, parameterName=UV Squash Min}
   B=ScalarParameter() {defaultValue=5, group=UVs, parameterName=UV Squash Max}
   Alpha=NamedRerouteUsage()
  ) {constA=0.2, constB=4}
  n184 = Add(A=Divide(A=ComponentMask(None=@n180) {b=true}, B=Multiply(A=@n182, B=@n183) {constB=2})) {constB=0.5}
  n185 = VertexInterpolator(VS=Abs(None=AppendVector(A=@n181, B=@n184)))
  NamedRerouteDeclaration(None=@n185.PS) {name=Cylindrical UV Stable, nodeColor=<rgba 0 1 0.2734 1>}
  n186 = ScalarParameter() {defaultValue=0.01, group=UVs, parameterName=Cylinder UV Min Size}
  n187 = ScalarParameter() {defaultValue=30, group=UVs, parameterName=Cylinder UV Max Size}
  n188 = NamedRerouteUsage()
  NamedRerouteDeclaration(None=Multiply(A=@n185.PS, B=LinearInterpolate(A=@n186, B=@n187, Alpha=@n188) {constA=1, constB=4})) {name=Cylindrical UV, nodeColor=<rgba 0 1 0.2734 1>}
  n189 = NamedRerouteUsage()
  n190 = Add(A=NamedRerouteUsage()) {constB=-0.5}
  n191 = OneMinus(None=NamedRerouteUsage())
  n192 = Add(A=@n189, B=Multiply(A=Multiply(A=@n189, B=@n190), B=Add(A=@n191) {constB=-0.5}))
  n193 = Divide(A=Time(), B=ScalarParameter() {defaultValue=7, group=Noises, parameterName=Wave Speed}) {constB=7}
  NamedRerouteDeclaration(None=Sine(None=Add(A=@n192, B=@n193))) {name=Wave Noise, nodeColor=<rgba 0 0.25 1 1>}
  NamedRerouteDeclaration(None=Saturate(None=@n36)) {name=Distance Mask Far, nodeColor=<rgba 0 1 0.8594 1>}
  n194 = CameraPositionWS()
  n195 = Multiply(A=WorldPosition().XYZ, B=Constant3Vector() {constant=<rgba 1 1 1 0>})
  n196 = ComponentMask(None=Divide(A=Subtract(A=@n194, B=@n195), B=ScalarParameter() {defaultValue=14317.7256, group=Gameplay Proximity, parameterName=Above Mask Falloff}) {constB=200}) {b=true}
  NamedRerouteDeclaration(None=OneMinus(None=Saturate(None=@n196))) {name=Above Mask, nodeColor=<rgba 0 0.8125 1 1>}
  NamedRerouteUsage()
  NamedRerouteUsage()
  NamedRerouteUsage()
  Saturate(None=@n2)
  ScalarParameter() {defaultValue=15000, parameterName=Debug radius}
  ScalarParameter() {defaultValue=0.9, parameterName=Debug Scale, sliderMax=2}
  VertexInterpolator()
  VertexInterpolator()
 ]
}
